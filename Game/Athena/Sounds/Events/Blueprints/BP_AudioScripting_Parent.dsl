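; Parent Blueprint that scripts audio for Fortnite live events, including the pre-event countdown.
; Delta-gated setters start the event, optionally after a timed offset, and EventUpdate toggles master-compressor bypass, submix audio analysis and a control bus mix together.
; Countdown timer delegates drive tick and alert sounds at the hologram.
; Reverb is set separately, sequence time feeds analysis stem components, and EndPlay cleans up event audio state.
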(event EventBeginPlay
  (CallFunction|SetCountdownTimer))

(event EventEndPlay
  (params (EndPlayReason :type EEndPlayReason%20Enum :default "Destroyed"))
  ;[1 1 1] Stop things that should not persist if player leaves events.
  (CallFunction|EventUpdate false)
  (Countdown|StopCountdownAudio)
  (CallFunction|InvalidStartEventHandle)
  (CallFunction|DeactivateReverb)
  (CallFunction|SetAudioAnalysis false))

(event Custom|SetActiveEvent
  (params (ActiveEvent :type Boolean))
  ;[1 1 1] Call this when event starts (this can be called from sequencer or another bp like master phase).
  (bind _updated_variable (Utilities|FlowControl|DeltaGateWildcard (Variables|Event|GetActiveEvent) ActiveEvent)
    (:Success
      (if _updated_variable
        (if (> (Variables|Event|GetEventStartOffsetTime) 0)
          (bind _returnvalue (Utilities|Time|SetTimerbyEvent (EventDispatchers|CreateEvent :!event Custom|StartEventTimer) (Variables|Event|GetEventStartOffsetTime)))
          (Variables|Countdown|SetStartEventHandle _returnvalue)
          (Countdown|PlayCountdownAlertAudio)
          (else
            (CallFunction|EventUpdate (Variables|Event|GetActiveEvent))
            (Countdown|StopCountdownAudio)
            (CallFunction|InvalidStartEventHandle)))))))

(event Custom|StartEventTimer
  ;[1 1 1] Call this when event starts (this can be called from sequencer or another bp like master phase).
  (CallFunction|EventUpdate (Variables|Event|GetActiveEvent))
  (Countdown|StopCountdownAudio)
  (CallFunction|InvalidStartEventHandle))

(event Custom|SetDeactiveEvent
  (params (DeactiveEvent :type Boolean))
  ;[1 1 1] Call this when event starts (this can be called from sequencer or another bp like master phase).
  (bind _updated_variable (Utilities|FlowControl|DeltaGateWildcard (Variables|Event|GetDeactiveEvent) DeactiveEvent)
    (:Success
      (if _updated_variable
        (CallFunction|EventUpdate false)))))

(event Custom|SetEventMix
  (params (EventMix :type Boolean))
  ;[1 1 1] Add or remoe Control Bus Mix
  (bind _updated_variable (Utilities|FlowControl|DeltaGateWildcard (Variables|Mix|GetEventMix) EventMix)
    (:Success
      (if _updated_variable
        (bind _sound_control_bus_mix (Variables|Mix|GetSound%20Control%20Bus%20Mix)
          (:then
            (if (not (Audio|Modulation|IsControlBusMixActive _sound_control_bus_mix))
              (Audio|Modulation|ActivateControlBusMix _sound_control_bus_mix)
              (Development|PrintString "Event Control Bus: Activated")))
          (:else))
        (else
          (bind _sound_control_bus_mix_2 (Variables|Mix|GetSound%20Control%20Bus%20Mix)
            (:then
              (if (Audio|Modulation|IsControlBusMixActive _sound_control_bus_mix_2)
                (Audio|Modulation|DeactivateControlBusMix _sound_control_bus_mix_2)
                (Development|PrintString "Event Control Bus: Deactivated")))
            (:else)))))))

(event Custom|SetReverbEffect
  (params (ReverbEffect :type Reverb%20Effect%20Object%20Reference))
  ;[1 1 1] Ability to set reverb and change it based on player experience.
  (bind _updated_variable (Utilities|FlowControl|DeltaGateWildcard (Variables|Reverb|GetReverbEffect) ReverbEffect)
    (:Success
      (Utilities|IsValid _updated_variable
        (:Is%20Valid
          (Audio|ActivateReverbEffect _updated_variable (Variables|Reverb|GetReverbTagName) (Variables|Reverb|GetReverbPriority) 1.0))))))

(event Custom|SetAudioAnalysis
  (params (AudioAnalysis :type Boolean))
  ;[1 1 1] Initialize event audio analyzer.
  (Utilities|FlowControl|DeltaGateWildcard (Variables|AudioAnalysis|GetAudioAnalysis) AudioAnalysis
    (:Success
      (CallFunction|OnInitializedChanged (Variables|AudioAnalysis|GetAudioAnalysis)))))

(event Custom|SetDisableMasterCompressor
  (params (DisableMasterCompressor :type Boolean))
  (bind _compressor (Variables|Compressor|GetCompressor))
  ;[1 1 1] Disable Master Compressor for events.
  (bind _updated_variable (Utilities|FlowControl|DeltaGateWildcard (Variables|Compressor|GetDisableMasterCompressor) DisableMasterCompressor)
    (:Success
      (Development|PrintString (Utilities|String|SelectString "Master Compressor: Disabled" "Master Compressor: Enabled" _updated_variable))
      (bind _structout (Utilities|Struct|SetmembersinSubmixEffectDynamicsProcessorSettings (Variables|/Script/AudioMixer.SubmixEffectDynamicsProcessorPreset|GetSettings _compressor) _updated_variable))
      (Audio|Effects|SetSettings _compressor _structout :!class /Script/AudioMixer.SubmixEffectDynamicsProcessorPreset))))

(event Custom|SetCountdownTimer
  ;[1 1 1] Countdown clock that shows up in game, before the actual event starts
  (if (Variables|Countdown|GetEnableCountTimerAudio)
    (bind _countdownasset (Variables|Countdown|GetCountdownAsset)
      (:then
        (bind _returnvalue (Actor|GetActorOfClass _countdownasset))
        (Utilities|IsValid _returnvalue
          (:Is%20Valid
            (bind _output_get (Variables|Countdown|SetCountDownRef _returnvalue))
            (Utilities|IsValid (Utilities|ResolveSoftReference (Variables|/Game/Athena/Events/Generic/CountdownTimer/BP_Generic_Countdown_Parent.BP_Generic_Countdown_Parent_C|GetTimerManagerSoftRef _output_get))
              (:Is%20Valid
                (Default|AssignOnSecondsRemainingUpdated (Utilities|ResolveSoftReference (Variables|/Game/Athena/Events/Generic/CountdownTimer/BP_Generic_Countdown_Parent.BP_Generic_Countdown_Parent_C|GetTimerManagerSoftRef _output_get)) (EventDispatchers|CreateEvent :!event OnCountdownUpdate))
                (Default|AssignOnTimerFinished (Utilities|ResolveSoftReference (Variables|/Game/Athena/Events/Generic/CountdownTimer/BP_Generic_Countdown_Parent.BP_Generic_Countdown_Parent_C|GetTimerManagerSoftRef _output_get)) (EventDispatchers|CreateEvent :!event OnCountdownFinished)))))))
      (:else))))

(fn CountdownStop
  (Countdown|StopCountdownAudio))

(fn CountdownLocation
  (bind _countdownref (Variables|Countdown|GetCountDownRef)
    (:then
      (bind _hologramcountdown (Variables|/Game/Athena/Events/Generic/CountdownTimer/BP_Generic_Countdown_Parent.BP_Generic_Countdown_Parent_C|GetHologramCountdown _countdownref)
        (:then
          (return (+ (Transformation|GetWorldLocation _hologramcountdown) (Variables|Countdown|GetCountdownOffsetLocation))))
        (:else)))
    (:else)))

(fn DeactivateReverb
  (Variables|Reverb|GetReverbEffect
    (:then
      (Audio|DeactivateReverbEffect (Variables|Reverb|GetReverbTagName)))
    (:else)))

(fn EventUpdate
  (params (bIsActive :type Boolean))
  (CallFunction|SetDisableMasterCompressor bIsActive)
  (CallFunction|SetAudioAnalysis bIsActive)
  (CallFunction|SetEventMix bIsActive))

(fn InvalidStartEventHandle
  (bind _starteventhandle (Variables|Countdown|GetStartEventHandle))
  (if (Utilities|Time|IsValidTimerHandle _starteventhandle)
    (Utilities|Time|ClearandInvalidateTimerbyHandle _starteventhandle)))

(fn OnCountdownFinished
  (if (Variables|Countdown|GetStopCountdownOnFinish)
    (Countdown|StopCountdownAudio)))

(fn OnCountdownUpdate
  (params (SecondsRemaining :type Integer))
  ;[1 1 1] Call audio if there should be sound based on countdown clock that we use for events.
  ;;[1 1 1] Get clock reference and make sure they are valid
  (bind _countdownref (Variables|Countdown|GetCountDownRef)
    (:then
      (bind _hologramcountdown (Variables|/Game/Athena/Events/Generic/CountdownTimer/BP_Generic_Countdown_Parent.BP_Generic_Countdown_Parent_C|GetHologramCountdown _countdownref)
        (:then
          (if (Math|Boolean|ORBoolean (not (Utilities|IsValid _countdownref :!class /Script/Engine.KismetSystemLibrary)) (not (Rendering|IsVisible _hologramcountdown)) (<= SecondsRemaining 0))
            (Utilities|Time|SetTimerbyEvent (EventDispatchers|CreateEvent :!event Countdown%20Stop) (Variables|Countdown|GetCountdownStopOffsetTime))
            (else
              (bind _countdownticksound (Variables|Countdown|GetCountdownTickSound)
                (:then
                  (Audio|PlaySoundatLocation _countdownticksound (Countdown|CountdownLocation)))
                (:else
                  _))
              (if (<= SecondsRemaining (Variables|Countdown|GetCountdownAlertThreshold))
                (Countdown|PlayCountdownAlertAudio)))))
        (:else)))
    (:else)))

(fn OnInitializedChanged
  (params (Initialized :type Boolean))
  (bind _returnvalue (WorldSubsystems|GetFortAudioAnalysisSubsystem))
  (bind _returnvalue_2 (WorldSubsystems|GetFortAudioAnalysisSubsystem))
  (Development|PrintString (Utilities|String|SelectString "AudioAnalysis: Inititlaized" "AudioAnalysis: Uninitialized" Initialized))
  (if Initialized
    (Utilities|FlowControl|Sequence
      (:then_0
        (if (Variables|AudioAnalysis|GetEnableAudioAnalysis_Default)
          (AudioAnalysis|RegisterDefaultSubmix _returnvalue_2)))
      (:then_1
        (if (Variables|AudioAnalysis|GetEnableAudioAnalysis_VO_A)
          (AudioAnalysis|RegisterSubmix _returnvalue_2 "/Game/Sounds/SubMixes/Fortnite_AudioAnalysis_VO_A.Fortnite_AudioAnalysis_VO_A" (Variables|AudioAnalysis|GetAnalyzerData_VO_A))))
      (:then_2
        (if (Variables|AudioAnalysis|GetEnableAudioAnalysis_VO_B)
          (AudioAnalysis|RegisterSubmix _returnvalue_2 "/Game/Sounds/SubMixes/Fortnite_AudioAnalysis_VO_B.Fortnite_AudioAnalysis_VO_B" (Variables|AudioAnalysis|GetAnalyzerData_VO_B))))
      (:then_3
        (if (Variables|AudioAnalysis|GetEnableAudioAnalysis_VO_C)
          (AudioAnalysis|RegisterSubmix _returnvalue_2 "/Game/Sounds/SubMixes/Fortnite_AudioAnalysis_VO_C.Fortnite_AudioAnalysis_VO_C" (Variables|AudioAnalysis|GetAnalyzerData_VO_C))))
      (:then_4
        (if (Variables|AudioAnalysis|GetEnableAudioAnalysis_Master)
          (AudioAnalysis|RegisterSubmix _returnvalue_2 "/Game/Sounds/SubMixes/Fortnite_Final.Fortnite_Final" (Variables|AudioAnalysis|GetAnalyzerData_VO_Master)))))
    (else
      ;[1 1 1] Unregister submixes if we no longer are analysing
      (Utilities|FlowControl|Sequence
        (:then_0
          (if (Variables|AudioAnalysis|GetEnableAudioAnalysis_Default)
            (AudioAnalysis|UnRegisterDefaultSubmix _returnvalue)))
        (:then_1
          (if (Variables|AudioAnalysis|GetEnableAudioAnalysis_VO_A)
            (AudioAnalysis|UnRegisterSubmix _returnvalue "/Game/Sounds/SubMixes/Fortnite_AudioAnalysis_VO_A.Fortnite_AudioAnalysis_VO_A")))
        (:then_2
          (if (Variables|AudioAnalysis|GetEnableAudioAnalysis_VO_B)
            (AudioAnalysis|UnRegisterSubmix _returnvalue "/Game/Sounds/SubMixes/Fortnite_AudioAnalysis_VO_B.Fortnite_AudioAnalysis_VO_B")))
        (:then_3
          (if (Variables|AudioAnalysis|GetEnableAudioAnalysis_VO_C)
            (AudioAnalysis|UnRegisterSubmix _returnvalue "/Game/Sounds/SubMixes/Fortnite_AudioAnalysis_VO_C.Fortnite_AudioAnalysis_VO_C")))
        (:then_4
          (if (Variables|AudioAnalysis|GetEnableAudioAnalysis_Master)
            (AudioAnalysis|UnRegisterSubmix _returnvalue "/Game/Sounds/SubMixes/Fortnite_Final.Fortnite_Final")))))))

(fn PlayCountdownAlertAudio
  ;[1 1 1] Update alert audio to play and set parameters if needed.
  (bind _countdownalertaudio (Variables|Default|GetCountdownAlertAudio)
    (:then)
    (:else
      (bind _countdownalertsound (Variables|Countdown|GetCountdownAlertSound)
        (:then
          (if (not (Audio|Components|Audio|IsPlaying _countdownalertaudio))
            (bind _returnvalue (Audio|SpawnSoundatLocation _countdownalertsound (Countdown|CountdownLocation)))
            (bind _output_get (Variables|Default|SetCountdownAlertAudio _returnvalue))
            (Audio|Parameter|SetBooleanParameter _output_get "WarningAlert" true :!class /Script/AudioExtensions.AudioParameterControllerInterface)))
        (:else)))))

(fn SetSequenceTime
  (params (SequenceTime :type Float%20%28double-precision%29))
  (Variables|AudioAnalysis|SetSequenceTime SequenceTime)
  (for _array_element (Actor|GetComponentsByClass "/Game/Athena/Events/AudioAnalysis/Events_AudioAnalysis_Stem_Component.Events_AudioAnalysis_Stem_Component_C")
    (Class|EventsAudioAnalysisStemComponent|Update _array_element (Variables|AudioAnalysis|GetSequenceTime))))

(fn StopCountdownAlertAudio
  (bind _countdownalertaudio (Variables|Default|GetCountdownAlertAudio)
    (:then
      (Audio|Components|Audio|Stop _countdownalertaudio))
    (:else)))

(fn StopCountdownAudio
  (Variables|Countdown|SetEnableCountTimerAudio false)
  (bind _countdownalertaudio (Variables|Default|GetCountdownAlertAudio)
    (:then
      (Audio|Parameter|SetBooleanParameter _countdownalertaudio "WarningAlert" false :!class /Script/AudioExtensions.AudioParameterControllerInterface)
      (EventDispatchers|BindEventtoOnAudioFinished _countdownalertaudio (EventDispatchers|CreateEvent :!event StopCountdownAlertAudio)))
    (:else)))

(fn ConstructionScript)
